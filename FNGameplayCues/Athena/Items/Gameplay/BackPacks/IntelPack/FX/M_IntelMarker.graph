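# Unlit translucent sprite particle material for the IntelPack backpack's intel marker, drawn with depth testing disabled.
# Particle color passes through HDR derivation and emissive adjustment functions for the glow.
# A soft radial UV mask is multiplied by Atlas scanlines with boosted secondary lines.
# That mask is multiplied by a world-space sphere mask around a PawnLocation parameter, confining it to an adjustable radius around that position.
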Material /FNGameplayCues/Athena/Items/Gameplay/BackPacks/IntelPack/FX/M_IntelMarker
graph {
 properties {
  bDisableDepthTest=true
  bUsedWithParticleSprites=true
  blendMode=BLEND_Translucent
  pixelDepthOffsetMode=PDOM_Legacy
  shadingModel=MSM_Unlit
  translucencyPass=MTP_BeforeDOF
  twoSided=true
  usedShadingModels=Unlit
 }
 nodes [
  n1 = ParticleColor()
  MaterialFunctionCall(Emissive=MaterialFunctionCall(LDR Input=Clamp(None=@n1.RGB), HDR Tint=@n1.RGB, HDR Intensity Multiplier=ScalarParameter() {defaultValue=5, parameterName=HDR_INtensity}) {materialFunction=&DeriveHDRfromLDR}.RemappedRange) {materialFunction=&MF_FX_EmissiveAdjustments}
  n2 = TextureCoordinate()
  n3 = Constant() {r=0.5}
  n4 = MaterialFunctionCall!(
   Number of Scanlines
   Scanline Speed
   "Scanlines Thickness (Bias)"
   Scanlines Dark Value=Constant()
   Scanlines Light Value
   Block Opacity=ScalarParameter() {defaultValue=0.25, parameterName=Block Opacity}
   UVs
   Time
   Block Repeat
  ) {materialFunction=/Game/Effects/Fort_Effects/Effects/Atlas/Materials/MF_Atlas_Scanlines_01}
  n5 = Add(A=@n4.Result, B=Multiply(A=@n4.Secondary Lines) {constB=100})
  n6 = VectorParameter() {parameterName=PawnLocation}
  Multiply(
   A=Multiply(A=SphereMask(A=@n2, B=@n3, Radius=@n3) {hardnessPercent=0}, B=LinearInterpolate(Alpha=@n5) {constB=0.5})
   B=SphereMask(
    A=WorldPosition().XYZ
    B=ComponentMask(None=@n6.RGB) {b=true, g=true, r=true}
    Radius=ScalarParameter() {defaultValue=200, parameterName=Radius}
   )
  )
 ]
}
